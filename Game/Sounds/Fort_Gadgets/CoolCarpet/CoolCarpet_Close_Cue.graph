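# Sound cue for the closing sound of the CoolCarpet gadget, classed as Athena player foley.
# A Random node picks one of three equally weighted close waves, which feeds a Modulator that applies a slight random volume boost.
# It uses the CoolCarpet footstep attenuation asset and overrides concurrency to allow only one instance at a time.
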SoundCue /Game/Sounds/Fort_Gadgets/CoolCarpet/CoolCarpet_Close_Cue
root=Modulator_0
nodes [
 {id=Modulator_0, children [Random_0], properties {volumeMax=1.6, volumeMin=1.5}}
 {id=Random_0, children [WavePlayer_1,WavePlayer_2,WavePlayer_3], properties {weights [1,1,1]}}
 {id=WavePlayer_1, properties {soundWaveAssetPtr=&/CoolCarpet_Close_01}}
 {id=WavePlayer_2, properties {soundWaveAssetPtr=&/CoolCarpet_Close_02}}
 {id=WavePlayer_3, properties {soundWaveAssetPtr=&/CoolCarpet_Close_03}}
]
properties {
 attenuationSettings=/Game/Sounds/Attenuation/Fort_Player_Footsteps_CoolCarpet
 bOverrideConcurrency=true
 maxDistance=3500
 soundClassObject=&Fort_Player_Foley_Athena
 volumeMultiplier=1.5
 concurrencyOverrides {maxCount=1}
}
