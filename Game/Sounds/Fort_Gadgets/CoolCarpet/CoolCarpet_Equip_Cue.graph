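# Sound cue for equipping the CoolCarpet gadget in Fortnite.
# A Random node picks one of two equip wave variations with equal weight, feeding a Modulator at the root.
# It plays in the weapon foley sound class with reduced volume and borrows the weapon reload attenuation.
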SoundCue /Game/Sounds/Fort_Gadgets/CoolCarpet/CoolCarpet_Equip_Cue
root=Modulator_1
nodes [
 {id=Modulator_1, children [Random_1]}
 {id=Random_1, children [WavePlayer_0,WavePlayer_4], properties {weights [1,1]}}
 {id=WavePlayer_0, properties {soundWaveAssetPtr=&/CoolCarpet_Equip_02}}
 {id=WavePlayer_4, properties {soundWaveAssetPtr=&/CoolCarpet_Equip_01}}
]
properties {
 attenuationSettings=&Fort_Weapon_Reload_Att
 maxDistance=2250
 soundClassObject=&Fort_WeaponFoley
 volumeMultiplier=0.401
}
