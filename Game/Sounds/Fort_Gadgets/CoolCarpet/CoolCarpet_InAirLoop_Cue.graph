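# Sound cue for the Cool Carpet gadget in flight, combining a throw one-shot with looping in-air audio.
# The loop layer mixes a heavily volume-boosted carpet loop with a borrowed Football toy in-air loop, under an envelope that quickly fades in.
# A random pick of two throw sounds plays alongside it without the fade.
# It uses close-range projectile attenuation and has the debug flag on.
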SoundCue /Game/Sounds/Fort_Gadgets/CoolCarpet/CoolCarpet_InAirLoop_Cue
root=Mixer_1
nodes [
 {
  id=Enveloper_0
  children [Mixer_0]
  properties {pitchCurve {editorCurveData {keys [{time=0}]}}, volumeCurve {editorCurveData {keys [<at 0 0.01>,<at 0.15 1>]}}}
 }
 {id=Mixer_0, children [Modulator_0,WavePlayer_1], properties {inputVolume [0.5,0.5]}}
 {id=Mixer_1, children [Enveloper_0,Random_0], properties {inputVolume [1,1]}}
 {id=Modulator_0, children [WavePlayer_0], properties {volumeMax=7, volumeMin=6}}
 {id=Random_0, children [WavePlayer_3,WavePlayer_4], properties {weights [1,1]}}
 {id=WavePlayer_0, properties {bLooping=true, soundWaveAssetPtr=&/CoolCarpet_Throw_Loop}}
 {id=WavePlayer_1, properties {bLooping=true, soundWaveAssetPtr=/Game/Athena/Sounds/Toys/Football/Toys_Football_Inair_Loop}}
 {id=WavePlayer_3, properties {soundWaveAssetPtr=&/CoolCarpet_Throw_01}}
 {id=WavePlayer_4, properties {soundWaveAssetPtr=&/CoolCarpet_Throw_02}}
]
properties {
 attenuationSettings=&Fort_Weapon_Projectile_Close_Athena_Att
 bDebug=true
 maxDistance=4950
 soundClassObject=&Fort_Gadgets_Items
}
