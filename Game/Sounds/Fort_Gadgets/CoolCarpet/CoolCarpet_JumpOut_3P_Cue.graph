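# Sound cue for the third-person jump-out of the CoolCarpet gadget, spatialized with the CoolCarpet player footsteps attenuation in the Athena player foley class.
# A mixer layers a random pick of two JumpOut waves, delayed by a short fixed time and boosted, over a quieter random pick of three Open waves that starts immediately.
# The mix passes through a modulator at default settings.
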SoundCue /Game/Sounds/Fort_Gadgets/CoolCarpet/CoolCarpet_JumpOut_3P_Cue
root=Modulator_2
nodes [
 {id=Delay_0, children [Random_0], properties {delayMax=0.1, delayMin=0.1}}
 {id=Mixer_0, children [Delay_0,Random_1], properties {inputVolume [1.25,0.75]}}
 {id=Modulator_2, children [Mixer_0]}
 {id=Random_0, children [WavePlayer_1,WavePlayer_2], properties {bShouldExcludeFromBranchCulling=true, weights [1,1]}}
 {id=Random_1, children [WavePlayer_3,WavePlayer_5,WavePlayer_6], properties {bShouldExcludeFromBranchCulling=true, weights [1,1,1]}}
 {id=WavePlayer_1, properties {soundWaveAssetPtr=&/CoolCarpet_JumpOut_01}}
 {id=WavePlayer_2, properties {soundWaveAssetPtr=&/CoolCarpet_JumpOut_02}}
 {id=WavePlayer_3, properties {soundWaveAssetPtr=&/CoolCarpet_Open_01}}
 {id=WavePlayer_5, properties {soundWaveAssetPtr=&/CoolCarpet_Open_02}}
 {id=WavePlayer_6, properties {soundWaveAssetPtr=&/CoolCarpet_Open_03}}
]
properties {
 attenuationSettings=/Game/Sounds/Attenuation/Fort_Player_Footsteps_CoolCarpet
 maxDistance=3500
 soundClassObject=&Fort_Player_Foley_Athena
 volumeMultiplier=1.8
}
