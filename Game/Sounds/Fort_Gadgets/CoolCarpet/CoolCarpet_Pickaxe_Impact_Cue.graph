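# Sound cue for the CoolCarpet gadget's pickaxe impact, played when a pickaxe hit lands.
# A random node picks one of three equally weighted hit waves and feeds it through a modulator at the root.
# The cue belongs to the Fort gadgets items sound class and borrows the grenade bounce attenuation for distance falloff.
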SoundCue /Game/Sounds/Fort_Gadgets/CoolCarpet/CoolCarpet_Pickaxe_Impact_Cue
root=Modulator_0
nodes [
 {id=Modulator_0, children [Random_1]}
 {id=Random_1, children [WavePlayer_1,WavePlayer_6,WavePlayer_7], properties {weights [1,1,1]}}
 {id=WavePlayer_1, properties {soundWaveAssetPtr=&/CoolCarpet_Hit_01}}
 {id=WavePlayer_6, properties {soundWaveAssetPtr=&/CoolCarpet_Hit_02}}
 {id=WavePlayer_7, properties {soundWaveAssetPtr=&/CoolCarpet_Hit_03}}
]
properties {
 attenuationSettings=/Game/Sounds/Attenuation/Fort_Grenade_Bounce_Att
 maxDistance=5500
 soundClassObject=&Fort_Gadgets_Items
 volumeMultiplier=1.0667
}
