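# Sound cue for the possess sound of the CoolCarpet gadget, played as player foley with CoolCarpet footstep attenuation.
# It mixes a quiet random Enter variant with a louder, slightly delayed random Possess variant, each chosen from two waves.
# The mix gets slight random pitch variation, and a short fixed delay holds back the whole cue.
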SoundCue /Game/Sounds/Fort_Gadgets/CoolCarpet/CoolCarpet_Possess_Cue
root=Delay_0
nodes [
 {id=Delay_0, children [Modulator_0], properties {delayMax=0.08, delayMin=0.08}}
 {id=Delay_1, children [Random_0], properties {delayMax=0.05, delayMin=0.05}}
 {id=Mixer_0, children [Random_1,Delay_1], properties {inputVolume [0.2,1.25]}}
 {id=Modulator_0, children [Mixer_0], properties {pitchMax=1.03, pitchMin=0.97}}
 {id=Random_0, children [WavePlayer_0,WavePlayer_2], properties {bRandomizeWithoutReplacement=false, weights [1,1]}}
 {id=Random_1, children [WavePlayer_1,WavePlayer_3], properties {bRandomizeWithoutReplacement=false, weights [1,1]}}
 {id=WavePlayer_0, properties {soundWaveAssetPtr=&/CoolCarpet_Possess_02}}
 {id=WavePlayer_1, properties {soundWaveAssetPtr=&/CoolCarpet_Enter_01}}
 {id=WavePlayer_2, properties {soundWaveAssetPtr=&/CoolCarpet_Possess_01}}
 {id=WavePlayer_3, properties {soundWaveAssetPtr=&/CoolCarpet_Enter_02}}
]
properties {
 attenuationSettings=/Game/Sounds/Attenuation/Fort_Player_Footsteps_CoolCarpet
 maxDistance=3500
 soundClassObject=&Fort_Player_Foley_Athena
 volumeMultiplier=0.85
}
