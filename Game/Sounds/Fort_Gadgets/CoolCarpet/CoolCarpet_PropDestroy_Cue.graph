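# Sound cue for the destruction of the CoolCarpet gadget prop, classed under gadget items.
# It mixes a Sneaky Snowman destroy wave with one of three randomly chosen, equally weighted CoolCarpet hit waves, with the hit branch boosted.
# An enveloper fades the borrowed snowman layer down over the first fraction of a second.
# The cue is spatialized with the shared grenade bounce attenuation.
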SoundCue /Game/Sounds/Fort_Gadgets/CoolCarpet/CoolCarpet_PropDestroy_Cue
root=Mixer_0
nodes [
 {
  id=Enveloper_0
  children [WavePlayer_1]
  properties {pitchCurve {editorCurveData {keys [{time=0}]}}, volumeCurve {editorCurveData {keys [<at 0 0.85>,<at 0.65 0.25>]}}}
 }
 {id=Mixer_0, children [Enveloper_0,Random_0], properties {inputVolume [1,2.5]}}
 {id=Random_0, children [WavePlayer_0,WavePlayer_2,WavePlayer_3], properties {weights [1,1,1]}}
 {id=WavePlayer_0, properties {soundWaveAssetPtr=&/CoolCarpet_Hit_03}}
 {id=WavePlayer_1, properties {soundWaveAssetPtr=/Game/Sounds/Fort_Gadgets/SneakySnowman/Version2/SneakySnowman_Destroy_01}}
 {id=WavePlayer_2, properties {soundWaveAssetPtr=&/CoolCarpet_Hit_01}}
 {id=WavePlayer_3, properties {soundWaveAssetPtr=&/CoolCarpet_Hit_02}}
]
properties {
 attenuationSettings=/Game/Sounds/Attenuation/Fort_Grenade_Bounce_Att
 maxDistance=5500
 soundClassObject=&Fort_Gadgets_Items
 volumeMultiplier=0.7
}
