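# Sound cue for the CoolCarpet gadget's throw tell, playing a whoosh sound.
# A Random node picks between two equally weighted whoosh waves and feeds a Modulator at the root.
# It routes to the Athena player foley sound class with a slight volume boost and borrows a CoolCarpet player-footsteps attenuation asset for distance falloff.
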SoundCue /Game/Sounds/Fort_Gadgets/CoolCarpet/CoolCarpet_Throw_Tell_Cue
root=Modulator_0
nodes [
 {id=Modulator_0, children [Random_1]}
 {id=Random_1, children [WavePlayer_0,WavePlayer_4], properties {weights [1,1]}}
 {id=WavePlayer_0, properties {soundWaveAssetPtr=&/CoolCarpet_Whoosh_01}}
 {id=WavePlayer_4, properties {soundWaveAssetPtr=&/CoolCarpet_Whoosh_02}}
]
properties {
 attenuationSettings=/Game/Sounds/Attenuation/Fort_Player_Footsteps_CoolCarpet
 maxDistance=3500
 soundClassObject=&Fort_Player_Foley_Athena
 volumeMultiplier=1.2
}
