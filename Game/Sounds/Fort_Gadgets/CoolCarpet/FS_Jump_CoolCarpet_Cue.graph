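# Sound cue for the jump footstep foley of the CoolCarpet gadget.
# A Random node picks one of four jump wave variations with equal weight, and a Modulator above it applies a narrow, boosted random volume range.
# The cue plays through the Athena player foley sound class with a CoolCarpet-specific footstep attenuation, and adds its own volume boost.
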SoundCue /Game/Sounds/Fort_Gadgets/CoolCarpet/FS_Jump_CoolCarpet_Cue
root=Modulator_0
nodes [
 {id=Modulator_0, children [Random_1], properties {volumeMax=1.9, volumeMin=1.8}}
 {id=Random_1, children [WavePlayer_0,WavePlayer_5,WavePlayer_6,WavePlayer_7], properties {weights [1,1,1,1]}}
 {id=WavePlayer_0, properties {soundWaveAssetPtr=&/FS_Jump_CoolCarpet_01}}
 {id=WavePlayer_5, properties {soundWaveAssetPtr=&/FS_Jump_CoolCarpet_02}}
 {id=WavePlayer_6, properties {soundWaveAssetPtr=&/FS_Jump_CoolCarpet_03}}
 {id=WavePlayer_7, properties {soundWaveAssetPtr=&/FS_Jump_CoolCarpet_04}}
]
properties {
 attenuationSettings=/Game/Sounds/Attenuation/Fort_Player_Footsteps_CoolCarpet
 maxDistance=3500
 soundClassObject=&Fort_Player_Foley_Athena
 volumeMultiplier=1.2
}
