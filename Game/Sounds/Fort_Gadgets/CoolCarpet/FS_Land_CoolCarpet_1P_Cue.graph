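# Sound cue for the first-person landing footstep sound of the CoolCarpet gadget.
# A Random node picks one of four equally weighted landing waves, and a Modulator varies its volume within a narrow boosted range.
# It plays in the Athena player foley sound class with a non-spatialized CoolCarpet footstep attenuation.
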SoundCue /Game/Sounds/Fort_Gadgets/CoolCarpet/FS_Land_CoolCarpet_1P_Cue
root=Modulator_0
nodes [
 {id=Modulator_0, children [Random_0], properties {volumeMax=1.85, volumeMin=1.7}}
 {id=Random_0, children [WavePlayer_1,WavePlayer_2,WavePlayer_3,WavePlayer_4], properties {weights [1,1,1,1]}}
 {id=WavePlayer_1, properties {soundWaveAssetPtr=&/FS_Land_CoolCarpet_01}}
 {id=WavePlayer_2, properties {soundWaveAssetPtr=&/FS_Land_CoolCarpet_02}}
 {id=WavePlayer_3, properties {soundWaveAssetPtr=&/FS_Land_CoolCarpet_03}}
 {id=WavePlayer_4, properties {soundWaveAssetPtr=&/FS_Land_CoolCarpet_04}}
]
properties {
 attenuationSettings=/Game/Sounds/Attenuation/Fort_Player_Footsteps_CoolCarpet_NonSpatialized
 maxDistance=3584
 soundClassObject=&Fort_Player_Foley_Athena
 volumeMultiplier=1.15
}
