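# Sound cue for first-person walking footsteps on the CoolCarpet gadget's surface, judging by its name and path.
# A Random node picks among eight equally weighted footstep waves and is excluded from branch culling.
# A Modulator at the root varies pitch and volume on each play.
# Output uses a non-spatialized footstep attenuation and the player foley sound class, with a raised volume multiplier.
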SoundCue /Game/Sounds/Fort_Gadgets/CoolCarpet/FS_Walk_CoolCarpet_1P_Cue
root=Modulator_0
nodes [
 {id=Modulator_0, children [Random_0], properties {pitchMax=1.1, pitchMin=0.9, volumeMax=1.25, volumeMin=0.7}}
 {
  id=Random_0
  children [WavePlayer_1, WavePlayer_2, WavePlayer_3, WavePlayer_4, WavePlayer_8, WavePlayer_9, WavePlayer_10, WavePlayer_11]
  properties {bShouldExcludeFromBranchCulling=true, weights [1,1,1,1,1,1,1,1]}
 }
 {id=WavePlayer_1, properties {soundWaveAssetPtr=&/FS_CoolCarpet_01}}
 {id=WavePlayer_10, properties {soundWaveAssetPtr=&/FS_CoolCarpet_07}}
 {id=WavePlayer_11, properties {soundWaveAssetPtr=&/FS_CoolCarpet_08}}
 {id=WavePlayer_2, properties {soundWaveAssetPtr=&/FS_CoolCarpet_02}}
 {id=WavePlayer_3, properties {soundWaveAssetPtr=&/FS_CoolCarpet_03}}
 {id=WavePlayer_4, properties {soundWaveAssetPtr=&/FS_CoolCarpet_04}}
 {id=WavePlayer_8, properties {soundWaveAssetPtr=&/FS_CoolCarpet_05}}
 {id=WavePlayer_9, properties {soundWaveAssetPtr=&/FS_CoolCarpet_06}}
]
properties {
 attenuationSettings=/Game/Sounds/Attenuation/Fort_Player_Footsteps_CoolCarpet_NonSpatialized
 maxDistance=3584
 soundClassObject=&Fort_Player_Foley_Athena
 volumeMultiplier=2
}
