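# Sound cue for third-person walking footsteps on the CoolCarpet gadget surface.
# A Random node picks one of eight equally weighted footstep waves and is excluded from branch culling.
# A Modulator varies pitch and volume on each play.
# Output is boosted, routed to the Athena player foley sound class, and spatialized with dedicated CoolCarpet footstep attenuation.
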SoundCue /Game/Sounds/Fort_Gadgets/CoolCarpet/FS_Walk_CoolCarpet_3P_Cue
root=Modulator_1
nodes [
 {id=Modulator_1, children [Random_1], properties {pitchMax=1.1, pitchMin=0.9, volumeMax=1.25, volumeMin=0.7}}
 {
  id=Random_1
  children [WavePlayer_0, WavePlayer_5, WavePlayer_6, WavePlayer_7, WavePlayer_12, WavePlayer_13, WavePlayer_14, WavePlayer_15]
  properties {bShouldExcludeFromBranchCulling=true, weights [1,1,1,1,1,1,1,1]}
 }
 {id=WavePlayer_0, properties {soundWaveAssetPtr=&/FS_CoolCarpet_01}}
 {id=WavePlayer_12, properties {soundWaveAssetPtr=&/FS_CoolCarpet_05}}
 {id=WavePlayer_13, properties {soundWaveAssetPtr=&/FS_CoolCarpet_06}}
 {id=WavePlayer_14, properties {soundWaveAssetPtr=&/FS_CoolCarpet_07}}
 {id=WavePlayer_15, properties {soundWaveAssetPtr=&/FS_CoolCarpet_08}}
 {id=WavePlayer_5, properties {soundWaveAssetPtr=&/FS_CoolCarpet_02}}
 {id=WavePlayer_6, properties {soundWaveAssetPtr=&/FS_CoolCarpet_03}}
 {id=WavePlayer_7, properties {soundWaveAssetPtr=&/FS_CoolCarpet_04}}
]
properties {
 attenuationSettings=/Game/Sounds/Attenuation/Fort_Player_Footsteps_CoolCarpet
 maxDistance=3500
 soundClassObject=&Fort_Player_Foley_Athena
 volumeMultiplier=2.5
}
